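# Sign-in screen widget for Epic account login, a Blueprint subclass of the native FortSignInWidget.
# A safe-zone layout stacks the Epic logo, email and masked password fields, a one-time-password button, and a centered Sign In button inside a named slot.
# The inputs and buttons are inherited from the parent class, and Sign In is bound to the Login input action.
# The only graph logic fires an element-interacted event when Sign In is clicked.
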WidgetBlueprint /FortniteCoreUI/UI/Login/SignIn/SignInWidget
graphs [EventGraph {nodes [
   n1 = "On Button Clicked (Button_SignIn)"!() -> (OutputDelegate, then, Button:Common Button Legacy Object Reference) {is=title K2Node_ComponentBoundEvent}
   > Fire Event Element Interacted(Element=@n1.Button)
  ]}]
variables []
widgets {
 info {inheritedWidgetCount=6, parentClass=/Script/FortniteUI.FortSignInWidget, rootWidgetClass=&SafeZone, widgetCount=16}
 widgets [
  SafeZone_0 {widgetClassPath=&SafeZone, widget=SafeZone_0}
  SizeBox_1 {
   widgetClassPath=&SizeBox
   parent=SafeZone_0
   widget=SizeBox_1
   properties {bOverride_MinDesiredWidth=true, minDesiredWidth=1300}
   slot {class=SafeZoneSlot, properties {hAlign=HAlign_Center}}
  }
  VerticalBox_1 {widgetClassPath=&VerticalBox, parent=SizeBox_1, widget=VerticalBox_1, slot {class=SizeBoxSlot}}
  ContentBox {
   widgetClassPath=&VerticalBox
   bIsVariable=true
   parent=VerticalBox_1
   widget=ContentBox
   slot {class=VerticalBoxSlot, properties {verticalAlignment=VAlign_Center, size {sizeRule=Fill, value=1}}}
  }
  CommonLazyImage_0 {
   widgetClassPath=/Script/CommonUI.CommonLazyImage
   bIsVariable=true
   parent=ContentBox
   widget=CommonLazyImage_0
   properties {brush {
     bIsDynamicallyLoaded=false
     drawAs=Image
     imageType=NoImage
     mirroring=NoMirror
     resourceName=None
     resourceObject=/FortniteCoreUI/UI/Login/Art/M_EpicLogo_DF
     tiling=NoTile
     imageSize=<200 200>
     margin=<ltrb 0 0 0 0>
     outlineSettings {
      bUseBrushTransparency=false
      roundingType=HalfHeightRadius
      width=0
      color {colorUseRule=UseColor_Specified, specifiedColor=<rgba 0 0 0 0>}
      cornerRadii=<0 0 0 0>
     }
     tintColor {colorUseRule=UseColor_Specified, specifiedColor=<rgba 1 1 1 1>}
     uVRegion {bIsValid=false, max=<0 0>, min=<0 0>}
    }}
   slot {class=VerticalBoxSlot, properties {horizontalAlignment=HAlign_Center, verticalAlignment=VAlign_Top, padding=<ltrb 0 10 0 35>}}
  }
  SizeBox_0 {
   widgetClassPath=&SizeBox
   parent=ContentBox
   widget=SizeBox_0
   properties {bOverride_WidthOverride=true, widthOverride=800}
   slot {class=VerticalBoxSlot, properties {horizontalAlignment=HAlign_Center, verticalAlignment=VAlign_Center}}
  }
  CommonBorder_0 {
   widgetClassPath=&CommonBorder
   parent=SizeBox_0
   widget=CommonBorder_0
   properties {brushColor=<rgba 1 1 1 0.2>, padding=<ltrb 0 0 0 0>}
   slot {class=SizeBoxSlot}
  }
  VerticalBox_0 {
   widgetClassPath=&VerticalBox
   parent=CommonBorder_0
   widget=VerticalBox_0
   properties {navigation=&self:WidgetTree.VerticalBox_0.WidgetNavigation_0}
   slot {class=BorderSlot, properties {padding=<ltrb 0 0 0 0>}}
  }
  Border_Email {
   widgetClassPath=&CommonBorder
   bInherited=true
   bIsVariable=true
   parent=VerticalBox_0
   widget=Border_Email
   properties {
    background {
     bIsDynamicallyLoaded=false
     drawAs=Image
     imageType=NoImage
     mirroring=NoMirror
     resourceName=None
     resourceObject=null
     tiling=NoTile
     imageSize=<128 32>
     margin=<ltrb 0.1 0.45 0.9 0.5>
     outlineSettings {
      bUseBrushTransparency=false
      roundingType=HalfHeightRadius
      width=0
      color {colorUseRule=UseColor_Specified, specifiedColor=<rgba 0 0 0 0>}
      cornerRadii=<0 0 0 0>
     }
     tintColor {colorUseRule=UseColor_Specified, specifiedColor=<rgba 1 1 1 0.9>}
     uVRegion {bIsValid=false, max=<0 0>, min=<0 0>}
    }
    padding=<ltrb 10 10 10 10>
   }
   slot {class=VerticalBoxSlot, properties {padding=<ltrb 20 20 20 0>}}
  }
  EditText_Email {
   widgetClassPath=/Script/UMG.EditableText
   bInherited=true
   bIsVariable=true
   parent=Border_Email
   widget=EditText_Email
   properties {
    hintText=EMAIL
    keyboardType=Email
    widgetStyle {
     backgroundImageComposing {
      bIsDynamicallyLoaded=false
      drawAs=Image
      imageType=FullColor
      mirroring=NoMirror
      resourceName=../../../Engine/Content/Slate/Old/HyperlinkDotted.png
      resourceObject=/Game/UI/GreyBox/white
      tiling=Both
      imageSize=<8 8>
      margin=<ltrb 0 0 0 0.1875>
      outlineSettings {
       bUseBrushTransparency=false
       roundingType=HalfHeightRadius
       width=0
       color {colorUseRule=UseColor_Specified, specifiedColor=<rgba 0 0 0 0>}
       cornerRadii=<0 0 0 0>
      }
      tintColor {colorUseRule=UseColor_Specified, specifiedColor=<rgba 0.41 0.41 0.41 1>}
      uVRegion {bIsValid=false, max=<0 0>, min=<0 0>}
     }
     backgroundImageSelected {
      bIsDynamicallyLoaded=false
      drawAs=Image
      imageType=FullColor
      mirroring=NoMirror
      resourceName=../../../Engine/Content/Slate/Common/EditableTextSelectionBackground.png
      resourceObject=&white
      tiling=NoTile
      imageSize=<8 8>
      margin=<ltrb 0.25 0.25 0.25 0.25>
      outlineSettings {
       bUseBrushTransparency=false
       roundingType=HalfHeightRadius
       width=0
       color {colorUseRule=UseColor_Specified, specifiedColor=<rgba 0 0 0 0>}
       cornerRadii=<0 0 0 0>
      }
      tintColor {colorUseRule=UseColor_Specified, specifiedColor=<rgba 0.41 0.41 0.41 1>}
      uVRegion {bIsValid=false, max=<0 0>, min=<0 0>}
     }
     caretImage {
      bIsDynamicallyLoaded=false
      drawAs=Image
      imageType=FullColor
      mirroring=NoMirror
      resourceName=../../../Engine/Content/Slate/Old/White.png
      resourceObject=null
      tiling=NoTile
      imageSize=<16 16>
      margin=<ltrb 0 0 0 0>
      outlineSettings {
       bUseBrushTransparency=false
       roundingType=HalfHeightRadius
       width=0
       color {colorUseRule=UseColor_Specified, specifiedColor=<rgba 0 0 0 0>}
       cornerRadii=<0 0 0 0>
      }
      tintColor {colorUseRule=UseColor_Specified, specifiedColor=<rgba 1 1 1 1>}
      uVRegion {bIsValid=false, max=<0 0>, min=<0 0>}
     }
     colorAndOpacity {colorUseRule=UseColor_Specified, specifiedColor=<rgba 0 0 0 1>}
     font {
      bForceMonospaced=false
      bMaterialIsStencil=false
      fontMaterial=null
      fontObject=&NotoSans
      letterSpacing=0
      monospacedWidth=1
      size=32
      skewAmount=0
      typefaceFontName=NotoSans-Regular
      outlineSettings {
       bApplyOutlineToDropShadows=false
       bMiteredCorners=false
       bSeparateFillAlpha=false
       outlineMaterial=null
       outlineSize=0
       outlineColor=<rgba 0 0 0 1>
      }
     }
     hintColorAndOpacity {colorUseRule=UseColor_Foreground, specifiedColor=<rgba 1 0 1 1>}
     selectedBackgroundColor {colorUseRule=UseColor_Foreground, specifiedColor=<rgba 1 0 1 1>}
    }
   }
   slot {class=BorderSlot, properties {padding=<ltrb 10 10 10 10>}}
  }
  Border_Password {
   widgetClassPath=&CommonBorder
   bInherited=true
   bIsVariable=true
   parent=VerticalBox_0
   widget=Border_Password
   properties {
    background {
     bIsDynamicallyLoaded=false
     drawAs=Image
     imageType=NoImage
     mirroring=NoMirror
     resourceName=None
     resourceObject=null
     tiling=NoTile
     imageSize=<32 32>
     margin=<ltrb 0 0 0 0>
     outlineSettings {
      bUseBrushTransparency=false
      roundingType=HalfHeightRadius
      width=0
      color {colorUseRule=UseColor_Specified, specifiedColor=<rgba 0 0 0 0>}
      cornerRadii=<0 0 0 0>
     }
     tintColor {colorUseRule=UseColor_Specified, specifiedColor=<rgba 1 1 1 0.9>}
     uVRegion {bIsValid=false, max=<0 0>, min=<0 0>}
    }
    padding=<ltrb 10 10 10 10>
   }
   slot {class=VerticalBoxSlot, properties {padding=<ltrb 20 10 20 20>}}
  }
  EditText_Password {
   widgetClassPath=&EditableText
   bInherited=true
   bIsVariable=true
   parent=Border_Password
   widget=EditText_Password
   properties {
    hintText=EPIC PASSWORD
    isPassword=true
    widgetStyle {
     backgroundImageComposing {
      bIsDynamicallyLoaded=false
      drawAs=Image
      imageType=FullColor
      mirroring=NoMirror
      resourceName=../../../Engine/Content/Slate/Common/EditableTextSelectionBackground.png
      resourceObject=&white
      tiling=NoTile
      imageSize=<8 8>
      margin=<ltrb 0.25 0.25 0.25 0.25>
      outlineSettings {
       bUseBrushTransparency=false
       roundingType=HalfHeightRadius
       width=0
       color {colorUseRule=UseColor_Specified, specifiedColor=<rgba 0 0 0 0>}
       cornerRadii=<0 0 0 0>
      }
      tintColor {colorUseRule=UseColor_Specified, specifiedColor=<rgba 0.41 0.41 0.41 1>}
      uVRegion {bIsValid=false, max=<0 0>, min=<0 0>}
     }
     backgroundImageSelected {
      bIsDynamicallyLoaded=false
      drawAs=Image
      imageType=FullColor
      mirroring=NoMirror
      resourceName=../../../Engine/Content/Slate/Common/EditableTextSelectionBackground.png
      resourceObject=&white
      tiling=NoTile
      imageSize=<8 8>
      margin=<ltrb 0.25 0.25 0.25 0.25>
      outlineSettings {
       bUseBrushTransparency=false
       roundingType=HalfHeightRadius
       width=0
       color {colorUseRule=UseColor_Specified, specifiedColor=<rgba 0 0 0 0>}
       cornerRadii=<0 0 0 0>
      }
      tintColor {colorUseRule=UseColor_Specified, specifiedColor=<rgba 0.41 0.41 0.41 1>}
      uVRegion {bIsValid=false, max=<0 0>, min=<0 0>}
     }
     caretImage {
      bIsDynamicallyLoaded=false
      drawAs=Image
      imageType=FullColor
      mirroring=NoMirror
      resourceName=../../../Engine/Content/Slate/Old/White.png
      resourceObject=null
      tiling=NoTile
      imageSize=<16 16>
      margin=<ltrb 0 0 0 0>
      outlineSettings {
       bUseBrushTransparency=false
       roundingType=HalfHeightRadius
       width=0
       color {colorUseRule=UseColor_Specified, specifiedColor=<rgba 0 0 0 0>}
       cornerRadii=<0 0 0 0>
      }
      tintColor {colorUseRule=UseColor_Specified, specifiedColor=<rgba 1 1 1 1>}
      uVRegion {bIsValid=false, max=<0 0>, min=<0 0>}
     }
     colorAndOpacity {colorUseRule=UseColor_Specified, specifiedColor=<rgba 0 0 0 1>}
     font {
      bForceMonospaced=false
      bMaterialIsStencil=false
      fontMaterial=null
      fontObject=&NotoSans
      letterSpacing=0
      monospacedWidth=1
      size=32
      skewAmount=0
      typefaceFontName=NotoSans-Regular
      outlineSettings {
       bApplyOutlineToDropShadows=false
       bMiteredCorners=false
       bSeparateFillAlpha=false
       outlineMaterial=null
       outlineSize=0
       outlineColor=<rgba 0 0 0 1>
      }
     }
     hintColorAndOpacity {colorUseRule=UseColor_Foreground, specifiedColor=<rgba 1 0 1 1>}
     selectedBackgroundColor {colorUseRule=UseColor_Foreground, specifiedColor=<rgba 1 0 1 1>}
    }
   }
   slot {class=BorderSlot, properties {padding=<ltrb 10 10 10 10>}}
  }
  Button_ForgotPassword {
   widgetClassPath=&IconTextButton
   bInherited=true
   bIsVariable=true
   parent=VerticalBox_0
   widget=Button_ForgotPassword
   properties {
    buttonText=Get One-Time Password
    contentAlignment=HAlign_Center
    onButtonBaseClicked=(&self:WidgetTree.Button_ForgotPassword.HandleOnButtonClicked)
    onButtonBaseDoubleClicked=(&self:WidgetTree.Button_ForgotPassword.HandleOnButtonDoubleClicked)
    onButtonBaseHovered=(&self:WidgetTree.Button_ForgotPassword.HandleOnButtonHovered)
    onButtonBaseUnhovered=(&self:WidgetTree.Button_ForgotPassword.HandleOnButtonUnhovered)
    onSelectedChangedBase=(&self:WidgetTree.Button_ForgotPassword.HandleOnSelectedChanged)
    style=/FortUILibrary/Foundation/Buttons/ButtonStyles/ButtonStyle-MediumTransparentWithCues.ButtonStyle-MediumTransparentWithCues_C
   }
   slot {class=VerticalBoxSlot, properties {padding=<ltrb 20 0 20 20>}}
  }
  HorizontalBox_0 {
   widgetClassPath=&HorizontalBox
   parent=ContentBox
   widget=HorizontalBox_0
   slot {class=VerticalBoxSlot, properties {horizontalAlignment=HAlign_Center, verticalAlignment=VAlign_Center, padding=<ltrb 20 50 20 20>}}
  }
  ButtonSlot {
   widgetClassPath=&NamedSlot
   bIsVariable=true
   parent=HorizontalBox_0
   widget=ButtonSlot
   properties {bExposeOnInstanceOnly=true}
   slot {class=HorizontalBoxSlot}
  }
  Button_SignIn {
   widgetClassPath=&IconTextButton
   bInherited=true
   bIsVariable=true
   parent=ButtonSlot
   widget=Button_SignIn
   properties {
    contentAlignment=HAlign_Center
    minHeight=120
    minWidth=500
    onButtonBaseClicked=(&self:WidgetTree.Button_SignIn.HandleOnButtonClicked)
    onButtonBaseDoubleClicked=(&self:WidgetTree.Button_SignIn.HandleOnButtonDoubleClicked)
    onButtonBaseHovered=(&self:WidgetTree.Button_SignIn.HandleOnButtonHovered)
    onButtonBaseUnhovered=(&self:WidgetTree.Button_SignIn.HandleOnButtonUnhovered)
    onSelectedChangedBase=(&self:WidgetTree.Button_SignIn.HandleOnSelectedChanged)
    style=/FortUILibrary/Foundation/Buttons/ButtonStyles/ButtonStyle_Feature_L.ButtonStyle_Feature_L_C
    triggeringInputAction {dataTable=&InputActionDataTable, rowName=Login}
   }
   slot {class=PanelSlot}
  }
 ]
}
